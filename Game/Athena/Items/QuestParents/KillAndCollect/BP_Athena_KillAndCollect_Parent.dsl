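; Parent Blueprint for Athena kill-and-collect quest drops, collected by players with the active objective.
; Construction applies child-configured mesh, effects and audio; ConsolidatedQuestComponent events toggle per-controller visibility, and a sphere overlap toggles ambient particles.
; On quest success, clients play reward feedback for interacted drops and shrink others once the objective completes; OnRep_bHasInteracted replays this if replication arrives late.
; An optional timer shrinks and despawns drops.
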(event EventBeginPlay
  ;[1 1 1] Begin Play - executes on spawn of actor
  (Utilities|FlowControl|Sequence
    (:then_0
      ;;[1 1 1] If Test bool is disabled AND the quest item definition is not valid, hide the actor and print message to log, then destroy it
      (if (and (not (Utilities|IsValid (Class|ConsolidatedQuestComponent|GetQuestItemDefinition (Variables|Components|GetConsolidatedQuestComponent)) :!class /Script/Engine.KismetSystemLibrary)) (not (Variables|Design|GetbTestVisualsNoQuestRequired)))
        (Rendering|SetActorHiddenInGame true)
        (Development|PrintString "Quest Item Definition not valid on the dropped actor's ConsolidatedQuest component (was it set?)" true true "(R=1.000000,G=0.000000,B=0.350908,A=1.000000)" 5.0)
        (Actor|SetLifeSpan 1.0)))
    (:then_1
      ;;[1 1 1] If the actor hasn't been picked up in several seconds, destroy it gracefully
      (if (> (Variables|Default|GetDespawnDuration) 0.0)
        (Utilities|FlowControl|Delay (Variables|Default|GetDespawnDuration))
        (Audio|Components|Audio|FadeOut (Variables|Default|GetAmbientAudio) (Variables|Design|GetAmbientAudioFadeDuration) 0.0 "Linear" :!class /Script/Engine.AudioComponent)
        (bind (_direction _scaledowntime) (Timeline|T_ExpireCurve Play)
          (:Update
            (Transformation|SetWorldScale3D (Variables|Components|GetSM_Quest_DroppedItem) (Math|Vector|Lerp(Vector) (.scale (Transformation|GetRelativeTransform (Variables|Components|GetSM_Quest_DroppedItem))) "0.010000,0.010000,0.010000" _scaledowntime))
            (Transformation|SetRelativeScale3D (Variables|Default|GetP_AmbientFX) (Math|Vector|Lerp(Vector) (.scale (Transformation|GetRelativeTransform (Variables|Default|GetP_AmbientFX))) "0.100000,0.100000,0.100000"))
            (Transformation|SetRelativeScale3D (Variables|Default|GetN_AmbientFX) (Math|Vector|Lerp(Vector) (.scale (Transformation|GetRelativeTransform (Variables|Default|GetN_AmbientFX))) "0.100000,0.100000,0.100000")))
          (:Finished
            (Components|Activation|Deactivate (Variables|Default|GetP_AmbientFX))
            (CallFunction|ToggleActorAndMeshHidden true)
            (Actor|SetLifeSpan 0.2))) ; Scale down mesh and ambient effect
      ))))

(event OnQuestReady(ConsolidatedQuestComponent)
  (params (PlayerController :type Fort%20Player%20Controller%20Object%20Reference))
  ;[1 1 1] OnQuestReady - If testing bool is enabled OR player has the indicated quest objective set in ConsolidatedQuest and it is active/not complete, show the actor
  (Variables|Default|SetFortPlayerController PlayerController)
  (if (not (Variables|Design|GetbTestVisualsNoQuestRequired))
    (bind _isactive (CallFunction|IsConsolidatedQuestObjectiveActive (Variables|Default|GetFortPlayerController)))
    (if _isactive
      (bind (_playercontrollerout _completedquestobjective) (Quest|HasPlayerCompletedQuestObjective (Variables|Default|GetFortPlayerController) (Class|ConsolidatedQuestComponent|GetQuestItemDefinition (Variables|Components|GetConsolidatedQuestComponent)) (Variables|/Script/FortniteGame.ConsolidatedQuestComponent|GetObjectiveBackendName (Variables|Components|GetConsolidatedQuestComponent))))
      (if _completedquestobjective
        _
        (else
          _))))
  (CallFunction|ToggleActorAndMeshHidden false (Variables|Default|GetFortPlayerController)) ; Actor and mesh should be visible for this player
)

(event OnQuestNotReady(ConsolidatedQuestComponent)
  (params (PlayerController :type Fort%20Player%20Controller%20Object%20Reference))
  ;[1 1 1] Default state for the actor - only show initially if testing bool is enabled
  (if (Variables|Design|GetbTestVisualsNoQuestRequired)
    (CallFunction|ToggleActorAndMeshHidden false PlayerController)
    (Development|PrintString "Quest testing bool enabled for dropped item" true true "(R=1.000000,G=0.000000,B=0.245293,A=1.000000)")
    (else
      (CallFunction|ToggleActorAndMeshHidden true PlayerController))))

(event Interaction|EventBlueprintOnInteract
  (params (InteractingPawn :type Fort%20Pawn%20Object%20Reference) (InteractionBeingAttempted :type EInteractionBeingAttempted%20Enum :default "FirstInteraction"))
  ;[1 1 1] Set replicated variable bHasInteracted when the player interacts so it can handle each spawned instance of an item accordingly
  (Variables|Default|SetbHasInteracted true)
  (if (not (Utilities|Array|ContainsItem (Variables|Default|GetInteractedPlayerPawns) InteractingPawn))
    (Utilities|Array|Add (Variables|Default|GetInteractedPlayerPawns) InteractingPawn)))

(event OnQuestComponentSuccessEvent(ConsolidatedQuestComponent)
  (params (PlayerController :type Fort%20Player%20Controller%20Object%20Reference))
  ;[1 1 1] Play VFX when the player collects the item. For all other items lying about, if the player has completed the objective, make them disappear gracefully
  (if (not (Networking|IsServer))
    (Variables|Default|SetbClientQuestSuccessEventCalled true))
  (if (Variables|Default|GetbHasInteracted)
    (Networking|SwitchHasAuthority
      (:Remote
        (Audio|Components|Audio|FadeOut (Variables|Default|GetAmbientAudio) (Variables|Design|GetAmbientAudioFadeDuration) 0.0 "Linear" :!class /Script/Engine.AudioComponent)
        (Class|AthenaQuestFunctionLibrary|AthenaPlayQuestRewardFeedback false (Variables|Default|GetSuccessSound) (Variables|Default|GetSuccessNiagara) 0 self (Game|Player|GetPlayerPawn PlayerController))
        (CallFunction|ToggleActorAndMeshHidden true PlayerController)
        (Components|Activation|Deactivate (Variables|Default|GetAmbientAudio)))) ; Make sure dropped item is only visible on the client
    (else
      (bind (_playercontrollerout _completedquestobjective) (Quest|HasPlayerCompletedQuestObjective PlayerController (Class|ConsolidatedQuestComponent|GetQuestItemDefinition (Variables|Components|GetConsolidatedQuestComponent)) (Variables|/Script/FortniteGame.ConsolidatedQuestComponent|GetObjectiveBackendName (Variables|Components|GetConsolidatedQuestComponent))))
      (Networking|SwitchHasAuthority
        (:Remote
          (CallFunction|ToggleActorAndMeshHidden false PlayerController)
          (if _completedquestobjective
            (Audio|Components|Audio|FadeOut (Variables|Default|GetAmbientAudio) (Variables|Design|GetAmbientAudioFadeDuration) 0.0 "Linear" :!class /Script/Engine.AudioComponent)
            (bind (_direction _scaledowntime) (Timeline|T_HideMeshCurve Play)
              (:Update
                (Transformation|SetWorldScale3D (Variables|Components|GetSM_Quest_DroppedItem) (Math|Vector|Lerp(Vector) (.scale (Transformation|GetRelativeTransform (Variables|Components|GetSM_Quest_DroppedItem))) "0.010000,0.010000,0.010000" _scaledowntime))
                (Transformation|SetRelativeScale3D (Variables|Default|GetP_AmbientFX) (Math|Vector|Lerp(Vector) (.scale (Transformation|GetRelativeTransform (Variables|Default|GetP_AmbientFX))) "0.100000,0.100000,0.100000" _scaledowntime))
                (Transformation|SetRelativeScale3D (Variables|Default|GetN_AmbientFX) (Math|Vector|Lerp(Vector) (.scale (Transformation|GetRelativeTransform (Variables|Default|GetN_AmbientFX))) "0.100000,0.100000,0.100000")))
              (:Finished
                (CallFunction|ToggleActorAndMeshHidden true PlayerController)
                (Components|Activation|Deactivate (Variables|Default|GetP_AmbientFX)))) ; Scale down mesh and ambient effect
          ))) ; Make sure dropped item is only visible on the client
    )))

(event Custom|ForceUpdate
  ;[1 1 1] Play VFX when the player collects the item. For all other items lying about, if the player has completed the objective, make them disappear gracefully
  (if (Variables|Default|GetbHasInteracted)
    (Networking|SwitchHasAuthority
      (:Remote
        (Audio|Components|Audio|FadeOut (Variables|Default|GetAmbientAudio) (Variables|Design|GetAmbientAudioFadeDuration) 0.0 "Linear" :!class /Script/Engine.AudioComponent)
        (Class|AthenaQuestFunctionLibrary|AthenaPlayQuestRewardFeedback false (Variables|Default|GetSuccessSound) (Variables|Default|GetSuccessNiagara) 0 self (Game|Player|GetPlayerPawn (stale OnQuestComponentSuccessEvent(ConsolidatedQuestComponent) PlayerController)))
        (CallFunction|ToggleActorAndMeshHidden true (stale OnQuestComponentSuccessEvent(ConsolidatedQuestComponent) PlayerController))
        (Components|Activation|Deactivate (Variables|Default|GetAmbientAudio)))) ; Make sure dropped item is only visible on the client
    (else
      (bind (_playercontrollerout _completedquestobjective) (Quest|HasPlayerCompletedQuestObjective (stale OnQuestComponentSuccessEvent(ConsolidatedQuestComponent) PlayerController) (Class|ConsolidatedQuestComponent|GetQuestItemDefinition (Variables|Components|GetConsolidatedQuestComponent)) (Variables|/Script/FortniteGame.ConsolidatedQuestComponent|GetObjectiveBackendName (Variables|Components|GetConsolidatedQuestComponent))))
      (Networking|SwitchHasAuthority
        (:Remote
          (CallFunction|ToggleActorAndMeshHidden false (stale OnQuestComponentSuccessEvent(ConsolidatedQuestComponent) PlayerController))
          (if _completedquestobjective
            (Audio|Components|Audio|FadeOut (Variables|Default|GetAmbientAudio) (Variables|Design|GetAmbientAudioFadeDuration) 0.0 "Linear" :!class /Script/Engine.AudioComponent)
            (Timeline|T_HideMeshCurve Play)))) ; Make sure dropped item is only visible on the client
    )))

(event OnComponentBeginOverlap(AmbientVFX_CollisionRadius)
  (params (OverlappedComponent :type Primitive%20Component%20Object%20Reference) (OtherActor :type Actor%20Object%20Reference) (OtherComp :type Primitive%20Component%20Object%20Reference) (OtherBodyIndex :type Integer) (bFromSweep :type Boolean) (SweepResult :type Hit%20Result%20Structure%20%28by%20ref%29))
  ;[1 1 1] Show Ambient VFX and play Ambient audio on actor if player is within the CollisionRadius and has the quest active, hide it when they leave
  (bind _asfort_player_pawn_athena (Utilities|Casting|CastToFortPlayerPawnAthena OtherActor))
  (Networking|SwitchHasAuthority
    (:Remote
      (Quest|HasPlayerCompletedQuest (Fortnite|GetFortPlayerControllerfromActor _asfort_player_pawn_athena) (Class|ConsolidatedQuestComponent|GetQuestItemDefinition (Variables|Components|GetConsolidatedQuestComponent)))
      (bind _isactive (CallFunction|IsConsolidatedQuestObjectiveActive (Fortnite|GetFortPlayerControllerfromActor _asfort_player_pawn_athena)))
      (if (or _isactive (Variables|Design|GetbTestVisualsNoQuestRequired))
        (Components|Activation|Activate (Variables|Default|GetP_AmbientFX) true)))))

(event OnComponentEndOverlap(AmbientVFX_CollisionRadius)
  (params (OverlappedComponent :type Primitive%20Component%20Object%20Reference) (OtherActor :type Actor%20Object%20Reference) (OtherComp :type Primitive%20Component%20Object%20Reference) (OtherBodyIndex :type Integer))
  ;[1 1 1] Show Ambient VFX and play Ambient audio on actor if player is within the CollisionRadius and has the quest active, hide it when they leave
  (Utilities|Casting|CastToFortPlayerPawnAthena OtherActor)
  (Networking|SwitchHasAuthority
    (:Remote
      (Components|Activation|Deactivate (Variables|Default|GetP_AmbientFX)))))

(fn BlueprintCanInteract
  (params (InteractingPawn :type Fort%20Pawn%20Object%20Reference) (InteractionBeingAttempted :type EInteractionBeingAttempted%20Enum :default "FirstInteraction") (InteractionType :type TInteractionType%20Enum :default "IT_NoInteraction"))
  (bind _returnvalue (Fortnite|GetFortPlayerControllerfromActor InteractingPawn))
  (bind _consolidatedquestcomponent (Variables|Components|GetConsolidatedQuestComponent))
  (bind _objectivebackendname (Variables|/Script/FortniteGame.ConsolidatedQuestComponent|GetObjectiveBackendName _consolidatedquestcomponent))
  ;[1 1 1] If player has the quest, objective is active, and they haven't already interacted, they can interact with the actor
  (if (Utilities|Array|ContainsItem (Variables|Default|GetInteractedPlayerPawns) InteractingPawn)
    _
    (elif (Quests|IsQuestObjectiveActive (Quests|GetQuestManager _returnvalue "Invalid") _objectivebackendname)
      (bind (_playercontrollerout _completedquestobjective) (Quest|HasPlayerCompletedQuestObjective _returnvalue (Class|ConsolidatedQuestComponent|GetQuestItemDefinition _consolidatedquestcomponent) _objectivebackendname))
      (if _completedquestobjective
        (return false)
        (else
          (return true)))
      (else
        (return false)))))

(fn BlueprintGetInteractionString
  (params (InteractingPawn :type Fort%20Pawn%20Object%20Reference) (InteractionBeingAttempted :type EInteractionBeingAttempted%20Enum :default "FirstInteraction"))
  ;[1 1 1] Change value of Append to Interact String var to reflect what item the player is collecting
  (return (Utilities|Text|FormatText "NSLOCTEXT(\"[8C961AEA4F9B0D2BB248FABACD1E9AC9]\", \"BFD648F54650D2FADC52C2821E8A2E1D\", \"Collect {item}\")" (Variables|Design|GetAppendToInteractText))))

(fn BlueprintGetInteractionTime
  (params (InteractingPawn :type Fort%20Pawn%20Object%20Reference) (InteractionBeingAttempted :type EInteractionBeingAttempted%20Enum :default "FirstInteraction"))
  (return 0.5 true))

(fn BlueprintOnLocalInteract
  (params (InteractingPawn :type Fort%20Player%20Pawn%20Object%20Reference) (InteractionBeingAttempted :type EInteractionBeingAttempted%20Enum :default "FirstInteraction"))
  (return false))

(fn IsConsolidatedQuestObjectiveActive
  (params (PlayerController :type Fort%20Player%20Controller%20Object%20Reference))
  (bind _returnvalue (Quests|GetQuestManager PlayerController "Invalid"))
  (Utilities|IsValid _returnvalue
    (:Is%20Valid
      (return (Quests|IsQuestObjectiveActive _returnvalue (Variables|/Script/FortniteGame.ConsolidatedQuestComponent|GetObjectiveBackendName (Variables|Components|GetConsolidatedQuestComponent)))))))

(fn OnRep_bHasInteracted
  (if (Variables|Default|GetbClientQuestSuccessEventCalled)
    (CallFunction|ForceUpdate)))

(fn ShouldDie
  (params (Damage :type Float%20%28single-precision%29) (EventInstigator :type Controller%20Object%20Reference) (DamageCauser :type Actor%20Object%20Reference) (bIsTest :type Boolean) (EffectContext :type Gameplay%20Effect%20Context%20Handle%20Structure%20%28by%20ref%29))
  (return false))

(fn ToggleActorandMeshHidden
  (params (ActorHidden :type Boolean) (PlayerController :type Fort%20Player%20Controller%20Object%20Reference))
  ;[1 1 1] Sets actor, mesh, and shadows to be disabled for the client if actor should be hidden.
  (Utilities|FlowControl|Sequence
    (:then_0
      (LocalVariables|SetActorHidden_Local ActorHidden)
      (bind _output_get (LocalVariables|SetPlayerController_Local PlayerController))
      (Utilities|IsValid _output_get
        (:Is%20Not%20Valid
          (return))) ; Make sure player controller passed in is valid
    )
    (:then_1
      (PlayerController|SetActorHiddenControllerOnly (LocalVariables|GetPlayerController_Local) (LocalVariables|GetActorInstance_Local) (LocalVariables|GetActorHidden_Local))
      (Rendering|SetVisibility (Variables|Components|GetSM_Quest_DroppedItem) (not (LocalVariables|GetActorHidden_Local)) true)
      (PlayerController|SetCompHiddenControllerOnly (LocalVariables|GetPlayerController_Local) (Variables|Components|GetSM_Quest_DroppedItem) (LocalVariables|GetActorHidden_Local) false)
      (Rendering|SetCastShadow (Variables|Components|GetSM_Quest_DroppedItem) (not (LocalVariables|GetActorHidden_Local))) ; If actor is hidden, do NOT cast a shadow
      (return))))

(fn ConstructionScript
  (bind _sm_droppeditem (Variables|Design|GetSM_DroppedItem))
  (bind _ambientcollisionvolumeradius (Variables|Design|GetAmbientCollisionVolumeRadius))
  (bind _ambientsound (Variables|Design|GetAmbientSound))
  (bind _ambientniagarasystem (Variables|Design|GetAmbientNiagaraSystem))
  ;[1 1 1] BGA Init
  (Utilities|FlowControl|Sequence
    (:then_0
      ;;[1 1 1] Set the static mesh data passed in from child class to the inherited static mesh component
      (Utilities|IsValid _sm_droppeditem
        (:Is%20Valid
          (Components|StaticMesh|SetStaticMesh (Variables|Components|GetSM_Quest_DroppedItem) _sm_droppeditem)
          (Transformation|SetRelativeScale3D (Variables|Components|GetSM_Quest_DroppedItem) (Variables|Design|GetMeshRelativeScale)))))
    (:then_1
      ;;[1 1 1] Initialize ambient particle effects and set them to auto activate if testing bool is enabled
      (Transformation|SetRelativeTransform (Variables|Default|GetP_AmbientFX) (Variables|Design|GetAmbientFX_Transform) false true)
      (Utilities|IsValid (Variables|Design|GetAmbientParticle)
        (:Is%20Valid
          (Effects|Components|ParticleSystem|SetTemplate (Variables|Default|GetP_AmbientFX) (Variables|Design|GetAmbientParticle)))
        (:Is%20Not%20Valid
          _))
      (Utilities|IsValid _ambientniagarasystem
        (:Is%20Valid
          (Niagara|SetNiagaraSystemAsset (Variables|Default|GetN_AmbientFX) _ambientniagarasystem true))
        (:Is%20Not%20Valid
          _))
      (if (Variables|Design|GetbTestVisualsNoQuestRequired)
        (Components|Activation|SetAutoActivate (Variables|Default|GetP_AmbientFX) true)))
    (:then_2
      ;;[1 1 1] Initialize ambient audio and set it to auto activate if testing bool is enabled
      (Utilities|IsValid _ambientsound
        (:Is%20Valid
          (Audio|Components|Audio|SetSound (Variables|Default|GetAmbientAudio) _ambientsound)
          (if (Variables|Design|GetbTestVisualsNoQuestRequired)
            (Components|Activation|SetAutoActivate (Variables|Default|GetAmbientAudio) true)))))
    (:then_3
      ;;[1 1 1] Set the sphere radius for the fade in/out of VFX and audio
      (if (> _ambientcollisionvolumeradius 0.0)
        (Components|Sphere|SetSphereRadius (Variables|Default|GetAmbientVFX_CollisionRadius) _ambientcollisionvolumeradius true)))))
